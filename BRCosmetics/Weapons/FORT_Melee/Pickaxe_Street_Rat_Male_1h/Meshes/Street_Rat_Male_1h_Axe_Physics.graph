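# Physics asset for the Street Rat Male one-handed pickaxe cosmetic mesh, giving its tassel a simulation-ready collision and joint setup.
# A box body covers the handle bone, and four small capsule bodies follow the dyn_tassel bone chain.
# Constraints link the capsules in sequence from the handle, with a wide swing at the root joint, tighter swing between segments and very little twist throughout.
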PhysicsAsset /BRCosmetics/Weapons/FORT_Melee/Pickaxe_Street_Rat_Male_1h/Meshes/Street_Rat_Male_1h_Axe_Physics
bodies {
 Melee_TwoHanded_Handle [{shapeType=Box, extentX=7.3929, extentY=8.4599, extentZ=47.3424, center=<0.4538 -8.1811 35.6499>, rotation=<rot 0 -90 0>}]
 dyn_tassel_01 [{shapeType=Capsule, length=4, radius=2, center=<4 0 0>, rotation=<rot 90 0 0>}]
 dyn_tassel_02 [{shapeType=Capsule, length=3, radius=2, center=<2 0 0>, rotation=<rot 90 0 0>}]
 dyn_tassel_03 [{shapeType=Capsule, length=3, radius=2, center=<2 0 0>, rotation=<rot 90 0 0>}]
 dyn_tassel_04 [{shapeType=Capsule, length=3, radius=2, center=<2 0 0>, rotation=<rot 90 0 0>}]
}
constraints [
 {bone1Name=dyn_tassel_02, bone2Name=dyn_tassel_01, swing1LimitDegrees=45, swing2LimitDegrees=45, twistLimitDegrees=5}
 {bone1Name=dyn_tassel_01, bone2Name=Melee_TwoHanded_Handle, swing1LimitDegrees=90, swing2LimitDegrees=90, twistLimitDegrees=5}
 {bone1Name=dyn_tassel_03, bone2Name=dyn_tassel_02, swing1LimitDegrees=45, swing2LimitDegrees=45, twistLimitDegrees=5}
 {bone1Name=dyn_tassel_04, bone2Name=dyn_tassel_03, swing1LimitDegrees=45, swing2LimitDegrees=45, twistLimitDegrees=5}
]
