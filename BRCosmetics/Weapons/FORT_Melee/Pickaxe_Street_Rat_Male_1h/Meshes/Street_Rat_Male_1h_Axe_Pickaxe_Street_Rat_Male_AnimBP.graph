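# Animation Blueprint for the Street Rat Male one-handed pickaxe that adds secondary motion to its skeleton, including a dangling tassel chain.
# The cached input pose feeds a RigidBody node used when bRigidBodyEnabled is true, with a Trail controller on the tassel as the fallback.
# Before the trail, bCanPlayCustomAnimations selects between two additive tassel bend presets, and a planar limit relative to the root bone constrains the chain.
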AnimBlueprint /BRCosmetics/Weapons/FORT_Melee/Pickaxe_Street_Rat_Male_1h/Meshes/Street_Rat_Male_1h_Axe_Pickaxe_Street_Rat_Male_AnimBP
target_skeleton=&/Street_Rat_Male_1h_Axe_Pickaxe_Street_Rat_Male_Skeleton
graphs [
 AnimGraph {nodes [
   n1 = RigidBody(ComponentPose=Local To Component(LocalPose=Use cached pose 'PrePose'())) {
    bClampLinearTranslationLimitToRefPose=true
    bOverrideWorldGravity=true
    overridePhysicsAsset=&/Street_Rat_Male_1h_Axe_Physics
    alphaBoolBlend {blendOption=Linear}
    baseBoneRef {boneName=b_AR_Root}
    componentAppliedLinearAccClamp=<6000 6000 6000>
    componentLinearAccScale=<0.45 0.45 0.5>
    componentLinearVelScale=<0.45 0.45 0.5>
    overrideWorldGravity {z=-980}
   }
   n2 = "Transform (Modify) Bone - Bone: dyn_tassel_01"(
    ComponentPose=Local To Component(LocalPose=Use cached pose 'PrePose'())
    Alpha=1
    Translation="0,0,0"
    Rotation="-55.000,0.000,0.000"
    Scale="1,1,1"
   ) {
    rotationMode=BMM_Additive
    rotationSpace=BCS_BoneSpace
    alphaBoolBlend {blendOption=Linear}
    boneToModify {boneName=dyn_tassel_01}
    rotation {pitch=-54.9999}
   }
   n3 = "Transform (Modify) Bone - Bone: dyn_tassel_02"(ComponentPose=@n2, Alpha=1, Translation="0,0,0", Rotation="-30,0.000,0.000", Scale="1,1,1") {
    rotationMode=BMM_Additive
    rotationSpace=BCS_BoneSpace
    translationMode=BMM_Additive
    translationSpace=BCS_BoneSpace
    alphaBoolBlend {blendOption=Linear}
    boneToModify {boneName=dyn_tassel_02}
    rotation {pitch=-30}
   }
   n4 = Component To Local(ComponentPose=@n3)
   n5 = "Transform (Modify) Bone - Bone: dyn_tassel_01"(ComponentPose=Local To Component(LocalPose=Use cached pose 'PrePose'()), Alpha=1, Translation="0,0,0", Rotation="-40,0,0", Scale="1,1,1") {
    rotationMode=BMM_Additive
    rotationSpace=BCS_BoneSpace
    alphaBoolBlend {blendOption=Linear}
    boneToModify {boneName=dyn_tassel_01}
    rotation {pitch=-40}
   }
   n6 = Component To Local(ComponentPose=@n5)
   n7 = Get bCanPlayCustomAnimations() -> (bCanPlayCustomAnimations:bool)
   n8 = Trail controller - Bone: dyn_tassel_04!(ComponentPose=Local To Component(LocalPose=Blend Poses by bool(BlendPose_0=@n6, BlendPose_1=@n4, bActiveValue=@n7) {blendType=Linear})) -> (Pose) {
    bUsePlanarLimit=true
    chainLength=4
    planarLimits [{drivingBone {boneName=b_AR_Root}, planeTransform {location=<0 0 -14>, rotation=<rot 0 0 -180>, scale=<1 1 1>}}]
    rotationLimits [{limitMax=<180 180 180>}, {limitMax=<180 180 180>}, {limitMax=<180 180 180>}, {limitMax=<180 180 180>}]
    rotationOffsets [<0 0 0>, <0 0 0>, <0 0 0>, <0 0 0>]
    trailBone {boneName=dyn_tassel_04}
    trailRelaxationSpeed {editorCurveData {keys [<at 0 12 RCIM_Linear>,<at 1 8 RCIM_Linear>]}}
   }
   Output Pose(Result=Blend Poses by bool(BlendPose_0=Component To Local(ComponentPose=@n1), BlendPose_1=Component To Local(ComponentPose=@n8), bActiveValue=Get bRigidBodyEnabled()) {blendType=Linear}) {layerGroup=None}
   Save cached pose 'PrePose'(Pose=Input Pose()) {cacheName=PrePose}
  ]}
 EventGraph {nodes [
   TryGetPawnOwner()
   Event BlueprintUpdateAnimation()
  ]}
]
variables []
